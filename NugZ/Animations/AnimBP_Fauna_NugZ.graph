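# Animation Blueprint for NugZ, a zombie variant of the Nug fauna creature, with a floppy neck.
# A state machine covers idle, movement, jumping, being held or thrown and swimming, beneath a tornado override and montage slots.
# Neck-side poses with additive head bobble layer on top, flipped by yaw rate or NeckAdjust curve and disabled during the alert montage.
# Look-at, slope warping, held root offset and leg IK finish the pose.
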AnimBlueprint /NugZ/Animations/AnimBP_Fauna_NugZ
target_skeleton=/Irwin/AI/Base/Nug_Bird_Invis_Skeleton
graphs [
 AnimGraph {nodes [
   n1 = Apply Additive(
    Base=Blend Poses by bool(
     BlendPose_0=Sequence Player 'NugZ_NeckR'() {sequence=&/NugZ_NeckR}
     BlendPose_1=Sequence Player 'NugZ_NeckL'() {sequence=&/NugZ_NeckL}
     BlendTime_0=0.3
     BlendTime_1=0.3
     bActiveValue=Get FlipNeckToRight() -> (FlipNeckToRight:bool)
    ) {blendType=Linear}
    Additive=Blendspace Player 'NugZ_HeadBobble_BS'!(X=Get Speed() -> (Speed:float)) -> (Pose) {blendSpace=&/NugZ_HeadBobble_BS}
   ) {alphaBoolBlend {blendOption=Linear}}
   n2 = Sequence Player 'Nug_Abduction'() {playRate=1.4, sequence=/Irwin/AI/Prey/Nug/Animations/Nug_Abduction}
   AnimGraphNode_StateMachine_0 = Locomotion() {is=title AnimGraphNode_StateMachine, maxTransitionsPerFrame=1}
   n3 = Get bIsInTornado() -> (bIsInTornado:bool)
   n4 = Slot 'FullBodyAdditiveSlot'(Source=Slot 'DefaultSlot'(Source=Blend Poses by bool(BlendPose_0=@n2, BlendPose_1=@AnimGraphNode_StateMachine_0, BlendTime_0=0.4, BlendTime_1=0.2, bActiveValue=@n3) {blendType=Linear, transitionType=Inertialization})) {slotName=FullBodyAdditiveSlot}
   n5 = Get TurnOffNeckFlop() -> (TurnOffNeckFlop:float)
   n6 = Weighted Look At - Bone: BipedHeadNeck_A_Head_C!(
    ComponentPose:Component Space Pose Link Structure=Local To Component(LocalPose=Layered blend per bone(BasePose=@n4, BlendPoses_0=@n1, BlendWeights_0=@n5) {layerSetup [{branchFilters [{boneName=BipedHeadNeck_A_NeckBase_C}]}]})
    Alpha:float=Get LookAtAlpha() -> (LookAtAlpha:float)
    LookAtLocation:Vector=Get AimTargetLocation() -> (AimTargetLocation:Vector)
   ) -> (Pose:Component Space Pose Link Structure) {
    is=title FortAnimGraphNode_WeightedLookAt
    interpolationTime=3
    interpolationType=EaseInOutExponent3
    lookAtClamp=50
    alphaBoolBlend {blendOption=Linear}
    boneToModify {boneName=BipedHeadNeck_A_Head_C}
    boneWeightsToApply [{weight=0.33}, {weight=0.33}, {weight=0.33}]
    lookAt_Axis {bInLocalSpace=false, axis {y=0}}
    lookUp_Axis {axis {x=0}}
   }
   n7 = Biped Slope Warping!(ComponentPose:Component Space Pose Link Structure=@n6, Alpha:float=1) -> (Pose:Component Space Pose Link Structure) {
    is=title FortAnimGraphNode_SlopeWarping
    maxStepHeight=35
    alphaBoolBlend {blendOption=Linear}
    feetDefinitions [
     {numBonesInLimb=3, fKFootBone {boneName=HoofedLeg_A_Ball_L}, iKFootBone {boneName=IK_Foot_L}}
     {numBonesInLimb=3, fKFootBone {boneName=HoofedLeg_A_Ball_R}, iKFootBone {boneName=IK_Foot_R}}
    ]
    iKFootRootBone {boneName=IK_Foot_Root}
    pelvisBone {boneName=QuadSpine_A_Pelvis_C}
   }
   n8 = "Transform (Modify) Bone - Bone: root"!(ComponentPose=@n7, bAlphaBoolEnabled:bool=Get bIsBeingHeld() -> (bIsBeingHeld:bool), Translation="0,0,29", Rotation="0,90,0") -> (Pose) {
    alphaInputType=Bool
    rotationMode=BMM_Additive
    rotationSpace=BCS_BoneSpace
    translationMode=BMM_Additive
    translationSpace=BCS_BoneSpace
    alphaBoolBlend {blendOption=Linear}
    boneToModify {boneName=root}
    rotation {yaw=90}
    translation {z=29}
   }
   n9 = Leg IK!(ComponentPose:Component Space Pose Link Structure=@n8, Alpha:float=1) -> (Pose:Component Space Pose Link Structure) {
    is=title AnimGraphNode_LegIK
    alphaBoolBlend {blendOption=Linear}
    legsDefinition [
     {
      bEnableKneeTwistCorrection=true
      bEnableRotationLimit=false
      footBoneForwardAxis=Y
      hingeRotationAxis=None
      minRotationAngle=15
      numBonesInLimb=3
      twistOffsetCurveName=None
      fKFootBone {boneName=HoofedLeg_A_Ball_L}
      iKFootBone {boneName=IK_Foot_L}
     }
     {
      bEnableKneeTwistCorrection=true
      bEnableRotationLimit=false
      footBoneForwardAxis=Y
      hingeRotationAxis=None
      minRotationAngle=15
      numBonesInLimb=3
      twistOffsetCurveName=None
      fKFootBone {boneName=HoofedLeg_A_Ball_R}
      iKFootBone {boneName=IK_Foot_R}
     }
    ]
   }
   Output Pose(Result=Inertialization!(Source:Pose Link Structure=Component To Local(ComponentPose=@n9)) -> (Pose:Pose Link Structure) {is=title AnimGraphNode_Inertialization}) {layerGroup=None}
   n10 = Layered blend per bone(
    BasePose=Sequence Player 'NugZ_IdleShort'() {sequence=&/NugZ_IdleShort}
    BlendPoses_0=Sequence Player 'Nug_Idle_Held'() {sequence=/Irwin/AI/Prey/Nug/Animations/Nug_Idle_Held}
    BlendWeights_0=Get LowerBodyMaskAlpha() -> (LowerBodyMaskAlpha:float)
   ) {layerSetup [{branchFilters [{blendDepth=2, boneName=QuadSpine_A_Pelvis_C}, {boneName=IK_Foot_Root}, {blendDepth=-1, boneName=QuadSpine_A_Spine1_C}]}]}
   Save cached pose 'HeldSavePose'(Pose=@n10) {cacheName=HeldSavePose}
  ]}
 Locomotion {
  entry=Idle
  path=AnimGraph.AnimGraphNode_StateMachine_0.Locomotion
  ids=id
  nodes [
   AnimStateNode_11 = IsBeingHeld() {
    is=title AnimStateNode
    stateEntered {guid=#1}
    stateFullyBlended {guid=#2}
    stateLeft {guid=#3}
   }
   AnimStateNode_13 = Swimming() {is=title AnimStateNode}
   AnimStateNode_15 = JumpStart() {
    is=title AnimStateNode
    stateEntered {guid=#4}
    stateFullyBlended {guid=#5}
    stateLeft {guid=#6}
   }
   AnimStateNode_17 = JumpLoop() {is=title AnimStateNode}
   AnimStateNode_18 = JumpLand() {is=title AnimStateNode}
   AnimStateNode_2 = Locomotion() {
    is=title AnimStateNode
    stateEntered {guid=#7}
    stateFullyBlended {guid=#8}
    stateLeft {guid=#9}
   }
   AnimStateNode_4 = Idle() {
    stateEntered {guid=#10}
    stateFullyBlended {guid=#11}
    stateLeft {guid=#12}
   }
   AnimStateNode_7 = ChickenChuck() {
    is=title AnimStateNode
    stateEntered {guid=#13}
    stateFullyBlended {guid=#14}
    stateLeft {guid=#15}
   }
   AnimStateNode_8 = ChuckLand() {
    is=title AnimStateNode
    stateEntered {guid=#16}
    stateFullyBlended {guid=#17}
    stateLeft {guid=#18}
   }
   AnimStateTransitionNode_0 = AnimStateTransitionNode() {crossfadeDuration=0.4, logicType=TLT_Inertialization, priorityOrder=3}
   AnimStateTransitionNode_10 = AnimStateTransitionNode() {bAutomaticRuleBasedOnSequencePlayerInState=true}
   AnimStateTransitionNode_12 = AnimStateTransitionNode() {crossfadeDuration=0.1, logicType=TLT_Inertialization}
   AnimStateTransitionNode_13 = AnimStateTransitionNode() {bAutomaticRuleBasedOnSequencePlayerInState=true, crossfadeDuration=0.25, logicType=TLT_Inertialization, priorityOrder=100}
   AnimStateTransitionNode_17 = AnimStateTransitionNode() {logicType=TLT_Inertialization}
   AnimStateTransitionNode_19 = AnimStateTransitionNode() {
    crossfadeDuration=0.15
    logicType=TLT_Inertialization
    transitionEnd {guid=#19}
    transitionInterrupt {guid=#20}
    transitionStart {guid=#21}
   }
   AnimStateTransitionNode_20 = AnimStateTransitionNode() {crossfadeDuration=0.4, logicType=TLT_Inertialization, priorityOrder=100}
   AnimStateTransitionNode_25 = AnimStateTransitionNode() {crossfadeDuration=0.4, logicType=TLT_Inertialization, priorityOrder=2}
   AnimStateTransitionNode_30 = AnimStateTransitionNode() {logicType=TLT_Inertialization, priorityOrder=100}
   AnimStateTransitionNode_32 = AnimStateTransitionNode() {logicType=TLT_Inertialization, priorityOrder=3}
   AnimStateTransitionNode_35 = AnimStateTransitionNode() {crossfadeDuration=0.4, logicType=TLT_Inertialization}
   AnimStateTransitionNode_37 = AnimStateTransitionNode() {crossfadeDuration=0.5, logicType=TLT_Inertialization}
   AnimStateTransitionNode_4 = AnimStateTransitionNode() {crossfadeDuration=0.4, logicType=TLT_Inertialization, priorityOrder=2}
   AnimStateTransitionNode_40 = AnimStateTransitionNode() {crossfadeDuration=0.3, logicType=TLT_Inertialization}
   AnimStateTransitionNode_42 = AnimStateTransitionNode() {bAutomaticRuleBasedOnSequencePlayerInState=true, crossfadeDuration=0.4, logicType=TLT_Inertialization}
   AnimStateTransitionNode_43 = AnimStateTransitionNode() {crossfadeDuration=0.4, logicType=TLT_Inertialization}
   AnimStateTransitionNode_5 = AnimStateTransitionNode() {
    crossfadeDuration=0.3
    logicType=TLT_Inertialization
    transitionEnd {guid=#22}
    transitionInterrupt {guid=#23}
    transitionStart {guid=#24}
   }
   AnimStateTransitionNode_7 = AnimStateTransitionNode() {crossfadeDuration=0.35, logicType=TLT_Inertialization}
  ]
  links [
   {
    dst=ChuckLand
    rule=AnimGraph.AnimGraphNode_StateMachine_0.Locomotion.AnimStateTransitionNode_5.Transition
    src=ChickenChuck
    via=AnimStateTransitionNode_5
   }
   {
    dst=Swimming
    rule=AnimGraph.AnimGraphNode_StateMachine_0.Locomotion.AnimStateTransitionNode_43.Transition
    src=ChickenChuck
    via=AnimStateTransitionNode_43
   }
   {
    dst=Idle
    rule=AnimGraph.AnimGraphNode_StateMachine_0.Locomotion.AnimStateTransitionNode_42.Transition
    src=ChuckLand
    via=AnimStateTransitionNode_42
   }
   {
    dst=IsBeingHeld
    rule=AnimGraph.AnimGraphNode_StateMachine_0.Locomotion.AnimStateTransitionNode_40.Transition
    src=Idle
    via=AnimStateTransitionNode_40
   }
   {
    dst=JumpStart
    rule=AnimGraph.AnimGraphNode_StateMachine_0.Locomotion.AnimStateTransitionNode_30.Transition
    src=Idle
    via=AnimStateTransitionNode_30
   }
   {
    dst=Locomotion
    rule=AnimGraph.AnimGraphNode_StateMachine_0.Locomotion.AnimStateTransitionNode_0.Transition
    src=Idle
    via=AnimStateTransitionNode_0
   }
   {
    dst=Swimming
    rule=AnimGraph.AnimGraphNode_StateMachine_0.Locomotion.AnimStateTransitionNode_37.Transition
    src=Idle
    via=AnimStateTransitionNode_37
   }
   {
    dst=ChickenChuck
    rule=AnimGraph.AnimGraphNode_StateMachine_0.Locomotion.AnimStateTransitionNode_19.Transition
    src=IsBeingHeld
    via=AnimStateTransitionNode_19
   }
   {
    dst=Idle
    rule=AnimGraph.AnimGraphNode_StateMachine_0.Locomotion.AnimStateTransitionNode_13.Transition
    src=JumpLand
    via=AnimStateTransitionNode_13
   }
   {
    dst=JumpStart
    rule=AnimGraph.AnimGraphNode_StateMachine_0.Locomotion.AnimStateTransitionNode_7.Transition
    src=JumpLand
    via=AnimStateTransitionNode_7
   }
   {
    dst=Locomotion
    rule=AnimGraph.AnimGraphNode_StateMachine_0.Locomotion.AnimStateTransitionNode_17.Transition
    src=JumpLand
    via=AnimStateTransitionNode_17
   }
   {
    dst=Swimming
    rule=AnimGraph.AnimGraphNode_StateMachine_0.Locomotion.AnimStateTransitionNode_4.Transition
    src=JumpLand
    via=AnimStateTransitionNode_4
   }
   {
    dst=JumpLand
    rule=AnimGraph.AnimGraphNode_StateMachine_0.Locomotion.AnimStateTransitionNode_12.Transition
    src=JumpLoop
    via=AnimStateTransitionNode_12
   }
   {
    dst=JumpLoop
    rule=AnimGraph.AnimGraphNode_StateMachine_0.Locomotion.AnimStateTransitionNode_10.Transition
    src=JumpStart
    via=AnimStateTransitionNode_10
   }
   {
    dst=Idle
    rule=AnimGraph.AnimGraphNode_StateMachine_0.Locomotion.AnimStateTransitionNode_25.Transition
    src=Locomotion
    via=AnimStateTransitionNode_25
   }
   {
    dst=JumpStart
    rule=AnimGraph.AnimGraphNode_StateMachine_0.Locomotion.AnimStateTransitionNode_32.Transition
    src=Locomotion
    via=AnimStateTransitionNode_32
   }
   {
    dst=Swimming
    rule=AnimGraph.AnimGraphNode_StateMachine_0.Locomotion.AnimStateTransitionNode_20.Transition
    src=Locomotion
    via=AnimStateTransitionNode_20
   }
   {
    dst=Locomotion
    rule=AnimGraph.AnimGraphNode_StateMachine_0.Locomotion.AnimStateTransitionNode_35.Transition
    src=Swimming
    via=AnimStateTransitionNode_35
   }
  ]
 }
 IsBeingHeld {
  path=AnimGraph.AnimGraphNode_StateMachine_0.Locomotion.AnimStateNode_11.IsBeingHeld
  nodes [
   n1 = "Transform (Modify) Bone - Bone: QuadSpine_A_Spine2_C"(ComponentPose=Local To Component(LocalPose=Use cached pose 'HeldSavePose'()), Alpha=1, Translation="0,0,0", Rotation="20,0,0", Scale="1,1,1") {rotationMode=BMM_Additive, alphaBoolBlend {blendOption=Linear}, boneToModify {boneName=QuadSpine_A_Spine2_C}, rotation {pitch=20}}
   n2 = "Transform (Modify) Bone - Bone: QuadSpine_A_Spine3_C"(ComponentPose=Local To Component(LocalPose=Use cached pose 'HeldSavePose'()), Alpha=1, Translation="0,0,0", Rotation="-9,0,0", Scale="1,1,1") {rotationMode=BMM_Additive, alphaBoolBlend {blendOption=Linear}, boneToModify {boneName=QuadSpine_A_Spine3_C}, rotation {pitch=-9}}
   n3 = "Transform (Modify) Bone - Bone: QuadSpine_A_Spine2_C"(ComponentPose=@n2, Alpha=1, Translation="0,0,0", Rotation="-9,0,0", Scale="1,1,1") {rotationMode=BMM_Additive, alphaBoolBlend {blendOption=Linear}, boneToModify {boneName=QuadSpine_A_Spine2_C}, rotation {pitch=-9}}
   Output Animation Pose(Result=Blend Poses by bool(
     BlendPose_0=Component To Local(ComponentPose=@n1)
     BlendPose_1=Component To Local(ComponentPose=@n3)
     BlendTime_0=0.4
     BlendTime_1=0.4
     bActiveValue=Get FlipNeckToRight()
    ) {blendType=Linear, transitionType=Inertialization})
  ]
 }
 Swimming {
  path=AnimGraph.AnimGraphNode_StateMachine_0.Locomotion.AnimStateNode_13.Swimming
  nodes [Output Animation Pose(Result=Sequence Player 'Nug_Swimming'() {sequence=/Irwin/AI/Prey/Nug/Animations/Nug_Swimming})]
 }
 JumpStart {
  path=AnimGraph.AnimGraphNode_StateMachine_0.Locomotion.AnimStateNode_15.JumpStart
  nodes [Output Animation Pose(Result=Sequence Player 'Nug_JumpStart'() {bLoopAnimation=false, sequence=/Irwin/AI/Prey/Nug/Animations/Nug_JumpStart})]
 }
 JumpLoop {
  path=AnimGraph.AnimGraphNode_StateMachine_0.Locomotion.AnimStateNode_17.JumpLoop
  nodes [Output Animation Pose(Result=Sequence Player 'Nug_JumpFall'() {sequence=/Irwin/AI/Prey/Nug/Animations/Nug_JumpFall})]
 }
 JumpLand {
  path=AnimGraph.AnimGraphNode_StateMachine_0.Locomotion.AnimStateNode_18.JumpLand
  nodes [Output Animation Pose(Result=Sequence Player 'Nug_JumpLand'() {bLoopAnimation=false, sequence=/Irwin/AI/Prey/Nug/Animations/Nug_JumpLand})]
 }
 Locomotion {
  path=AnimGraph.AnimGraphNode_StateMachine_0.Locomotion.AnimStateNode_2.Locomotion
  nodes [
   n1 = Get LeanDirection() -> (LeanDirection:float)
   n2 = Get MovingPlayRate() -> (MovingPlayRate:float)
   n3 = Get Speed()
   Output Animation Pose(Result=Apply Additive(
     Base=Blendspace Player 'NugZ_WalkRun_Zombie_BS'!(X=@n3, PlayRate:float=@n2) -> (Pose) {blendSpace=&/NugZ_WalkRun_Zombie_BS, groupName=Locomotion, method=SyncGroup}
     Additive=Blendspace Player 'Nug_Leans_BS'!(X=@n1) -> (Pose) {blendSpace=/Irwin/AI/Prey/Nug/Animations/Nug_Leans_BS}
    ) {alphaBoolBlend {blendOption=Linear}})
  ]
 }
 Idle {
  path=AnimGraph.AnimGraphNode_StateMachine_0.Locomotion.AnimStateNode_4.Idle
  nodes [
   n1 = Blendspace Player 'Nug_Leans_BS'(X=Get IdleShuffleBodyBend() -> (IdleShuffleBodyBend:float)) {blendSpace=&Nug_Leans_BS}
   n2 = Get IdleShufflePlayRate() -> (IdleShufflePlayRate:float)
   n3 = Layered blend per bone(
    BasePose=Sequence Player 'NugZ_IdleShort'() {sequence=&NugZ_IdleShort}
    BlendPoses_0=Blend Poses by bool(
     BlendPose_0=Sequence Player 'Nug_ShuffleTurnR'!(PlayRate:float=@n2) -> (Pose) {sequence=/Irwin/AI/Prey/Nug/Animations/Nug_ShuffleTurnR}
     BlendPose_1=Sequence Player 'Nug_ShuffleTurnL'!(PlayRate:float=@n2) -> (Pose) {sequence=/Irwin/AI/Prey/Nug/Animations/Nug_ShuffleTurnL}
     BlendTime_0=0.7
     BlendTime_1=0.7
     bActiveValue=Get bIdleShuffleTurnDirection() -> (bIdleShuffleTurnDirection:bool)
    ) {blendType=Linear, transitionType=Inertialization}
   ) {layerSetup [{branchFilters [{boneName=QuadSpine_A_Pelvis_C}, {blendDepth=-1, boneName=QuadSpine_A_Spine1_C}]}]}
   Output Animation Pose(Result=Blend Poses by bool(
     BlendPose_0=Apply Additive(Base=@n3, Additive=@n1) {alphaBoolBlend {blendOption=Linear}}
     BlendPose_1=Sequence Player 'NugZ_IdleShort'() {sequence=&NugZ_IdleShort}
     BlendTime_0=0.6
     BlendTime_1=0.6
     bActiveValue=Get bIsIdleShuffleTurn() -> (bIsIdleShuffleTurn:bool)
    ) {blendType=Linear, transitionType=Inertialization})
  ]
 }
 ChickenChuck {
  path=AnimGraph.AnimGraphNode_StateMachine_0.Locomotion.AnimStateNode_7.ChickenChuck
  nodes [Output Animation Pose(Result=Sequence Player 'Nug_JumpFall'() {sequence=&Nug_JumpFall})]
 }
 ChuckLand {
  path=AnimGraph.AnimGraphNode_StateMachine_0.Locomotion.AnimStateNode_8.ChuckLand
  nodes [Output Animation Pose(Result=Sequence Player 'Nug_JumpLand'() {bLoopAnimation=false, sequence=&Nug_JumpLand})]
 }
 Transition {
  path=AnimGraph.AnimGraphNode_StateMachine_0.Locomotion.AnimStateTransitionNode_0.Transition
  nodes [Result(bCanEnterTransition=Get bCanTransitionIdleToMoving() -> (bCanTransitionIdleToMoving:bool))]
 }
 Transition {path=AnimGraph.AnimGraphNode_StateMachine_0.Locomotion.AnimStateTransitionNode_10.Transition, nodes [Result()]}
 Transition {
  path=AnimGraph.AnimGraphNode_StateMachine_0.Locomotion.AnimStateTransitionNode_12.Transition
  nodes [Result(bCanEnterTransition=Get bCanTransitionJumpLoopToJumpLand() -> (bCanTransitionJumpLoopToJumpLand:bool))]
 }
 Transition {
  path=AnimGraph.AnimGraphNode_StateMachine_0.Locomotion.AnimStateTransitionNode_13.Transition
  nodes [
   Result(bCanEnterTransition=true)
   "float < float"!(A:double="Time Remaining (ratio) (Nug_JumpLand)"!() -> (ReturnValue:float) {is=title K2Node_AnimGetter}, B:double=0.1) -> (ReturnValue:bool) {is=title K2Node_CallFunction}
  ]
 }
 Transition {
  path=AnimGraph.AnimGraphNode_StateMachine_0.Locomotion.AnimStateTransitionNode_17.Transition
  nodes [Result(bCanEnterTransition=Get bCanTransitionIdleToMoving())]
 }
 Transition {
  path=AnimGraph.AnimGraphNode_StateMachine_0.Locomotion.AnimStateTransitionNode_19.Transition
  nodes [Result(bCanEnterTransition=Get bCanTransitionIsBeingHeldToChickenChuck() -> (bCanTransitionIsBeingHeldToChickenChuck:bool))]
 }
 Transition {
  path=AnimGraph.AnimGraphNode_StateMachine_0.Locomotion.AnimStateTransitionNode_20.Transition
  nodes [Result(bCanEnterTransition=Get bIsSurfaceSwimming())]
 }
 Transition {
  path=AnimGraph.AnimGraphNode_StateMachine_0.Locomotion.AnimStateTransitionNode_25.Transition
  nodes [Result(bCanEnterTransition=Get bCanTransitionMovingToIdle() -> (bCanTransitionMovingToIdle:bool))]
 }
 Transition {
  path=AnimGraph.AnimGraphNode_StateMachine_0.Locomotion.AnimStateTransitionNode_30.Transition
  nodes [Result(bCanEnterTransition=Get bIsJumping())]
 }
 Transition {
  path=AnimGraph.AnimGraphNode_StateMachine_0.Locomotion.AnimStateTransitionNode_32.Transition
  nodes [Result(bCanEnterTransition=Get bIsJumping())]
 }
 Transition {
  path=AnimGraph.AnimGraphNode_StateMachine_0.Locomotion.AnimStateTransitionNode_35.Transition
  nodes [Result(bCanEnterTransition=NOT Boolean(A=Get bIsSurfaceSwimming()))]
 }
 Transition {
  path=AnimGraph.AnimGraphNode_StateMachine_0.Locomotion.AnimStateTransitionNode_37.Transition
  nodes [Result(bCanEnterTransition=Get bIsSurfaceSwimming())]
 }
 Transition {
  path=AnimGraph.AnimGraphNode_StateMachine_0.Locomotion.AnimStateTransitionNode_4.Transition
  nodes [Result(bCanEnterTransition=Get bIsSurfaceSwimming())]
 }
 Transition {
  path=AnimGraph.AnimGraphNode_StateMachine_0.Locomotion.AnimStateTransitionNode_40.Transition
  nodes [Result(bCanEnterTransition=Get bIsBeingHeld())]
 }
 Transition {
  path=AnimGraph.AnimGraphNode_StateMachine_0.Locomotion.AnimStateTransitionNode_42.Transition
  nodes [
   Result(bCanEnterTransition=true)
   "float < float"(A="Time Remaining (ratio) (Nug_JumpLand)"(), B=0.1)
  ]
 }
 Transition {
  path=AnimGraph.AnimGraphNode_StateMachine_0.Locomotion.AnimStateTransitionNode_43.Transition
  nodes [Result(bCanEnterTransition=Get bIsSurfaceSwimming())]
 }
 Transition {
  path=AnimGraph.AnimGraphNode_StateMachine_0.Locomotion.AnimStateTransitionNode_5.Transition
  nodes [Result(bCanEnterTransition=NOT Boolean(A=Get bIsBeingHeld()))]
 }
 Transition {
  path=AnimGraph.AnimGraphNode_StateMachine_0.Locomotion.AnimStateTransitionNode_7.Transition
  nodes [Result(bCanEnterTransition=Get bIsJumping())]
 }
 EventGraph {nodes [
   Event BlueprintUpdateAnimation()
   > n1 = Sequence!(execute) -> (then_0, then_1, then_2:Exec, then_3:Exec, then_4:Exec)
   > n2 = Branch(Condition=Get bIsBeingHeld())
   > Set TurnLeft(TurnLeft:bool=false) -> (Output_Get:bool)
   > Set TurnRight(TurnRight:bool=false) -> (Output_Get:bool)
   n3 = Get PawnRotationRate() -> (PawnRotationRate:Rotator)
   n4 = Branch(execute=@n2.else, Condition="float >= float"(A=BreakRotator(InRot=@n3).Yaw, B=0.75))
   > Set TurnRight(TurnRight=true)
   n5 = Get PawnRotationRate()
   n6 = Branch(execute=@n4.else, Condition="float < float"(A=BreakRotator(InRot=@n5).Yaw, B=-0.75) {is=title K2Node_PromotableOperator})
   > Set TurnLeft(TurnLeft=true)
   Set TurnLeft(execute=@n6.else, TurnLeft=false)
   > Set TurnRight(TurnRight=false)
   n7 = Branch(execute=@n1.then_1, Condition="float >= float"(A=GetCurveValue(CurveName=NeckAdjust), B=0.8))
   > n8 = Branch(Condition=Get NeckLeft() -> (NeckLeft:bool))
   > n9 = Branch(Condition=Get IsPlayingAttack() -> (IsPlayingAttack:bool))
   Set FlipNeckToRight(execute=@n9.else, FlipNeckToRight:bool=true) -> (Output_Get:bool)
   > Set NeckLeft(NeckLeft:bool=false) -> (Output_Get:bool)
   > Set IsPlayingAttack(IsPlayingAttack:bool=true) -> (Output_Get:bool)
   n10 = Branch(execute=@n8.else, Condition=Get IsPlayingAttack())
   Set FlipNeckToRight(execute=@n10.else, FlipNeckToRight=false)
   > Set NeckLeft(NeckLeft=true)
   > Set IsPlayingAttack(IsPlayingAttack=true)
   Set IsPlayingAttack(execute=@n7.else, IsPlayingAttack=false)
   > n11 = Branch(Condition=Get TurnLeft() -> (TurnLeft:bool))
   > n12 = Branch(Condition=Get NeckLeft())
   > Set FlipNeckToRight(FlipNeckToRight=true)
   > Set NeckLeft(NeckLeft=false)
   Set FlipNeckToRight(execute=@n12.else, FlipNeckToRight=false)
   n13 = Branch(execute=@n11.else, Condition=Get TurnRight() -> (TurnRight:bool))
   > n14 = Branch(Condition=Get NeckLeft())
   > Set FlipNeckToRight(FlipNeckToRight=false)
   Set FlipNeckToRight(execute=@n14.else, FlipNeckToRight=true)
   > Set NeckLeft(NeckLeft=false)
   n15 = Branch(execute=@n13.else, Condition=Get FlipNeckToRight())
   > Set FlipNeckToRight(FlipNeckToRight=true)
   > Set NeckLeft(NeckLeft=false)
   Set FlipNeckToRight(execute=@n15.else, FlipNeckToRight=false)
   n16 = Branch(execute=@n1.then_2, Condition=Montage_IsPlaying(Montage=&/NugZ_Alert_Montage))
   > Set TurnOffNeckFlop(TurnOffNeckFlop:float=0) -> (Output_Get:float)
   Set TurnOffNeckFlop(execute=@n16.else, TurnOffNeckFlop=1)
   Event BlueprintInitializeAnimation()
   > Set IsPlayingAttack(IsPlayingAttack=false)
   > Set NeckFlopRondom(NeckFlopRondom:int=RandomIntegerInRange!(Min:int=0, Max:int=1) -> (ReturnValue:int=0)) -> (Output_Get:int=0)
   > n17 = Branch(Condition="Equal (Integer)"(A=Get NeckFlopRondom() -> (NeckFlopRondom:int=0)))
   > Set NeckLeft(NeckLeft=true)
   Set NeckLeft(execute=@n17.else, NeckLeft=false)
   n18 = Get PawnRotationRate()
   n19 = Branch(Condition="float >= float"(A=BreakRotator(InRot=@n18).Yaw, B=0.9))
   > Set TurnLeft(TurnLeft=true)
   n20 = Get PawnRotationRate()
   Branch(execute=@n19.else, Condition="float < float"(A=BreakRotator(InRot=@n20).Yaw, B=-0.9))
   > Set TurnRight(TurnRight=true)
  ]}
 Fauna Idle {nodes [
   Input Pose() {name=InPose Idle}
   Output Pose(Result=Sequence Player 'NugZ_IdleShortNeckZero'() {sequence=&/NugZ_IdleShortNeckZero}) {layerGroup=FaunaIdleOverride}
   Sequence Player 'NugZ_IdleShort'() {sequence=&NugZ_IdleShort}
  ]}
 Fauna IdleTurn {nodes [
   Input Pose() {name=InPose IdleTurn}
   Output Pose(Result=Sequence Player 'Nug_IdleShort'() {sequence=/Irwin/AI/Prey/Nug/Animations/Nug_IdleShort}) {layerGroup=FaunaIdleOverride}
  ]}
 Fauna Locomotion {nodes [Output Pose(Result=Input Pose() {name=InPose Locomotion}) {layerGroup=FaunaLocomotionOverride}]}
 Fauna LocomotionTurn {nodes [
   Input Pose() {name=InPose LocomotionTurn}
   n1 = Sequence Player 'Nug_TurnLeft90_InMotion'() {
    bLoopAnimation=false
    groupName=Locomotion
    groupRole=TransitionFollower
    method=SyncGroup
    sequence=/Irwin/AI/Prey/Nug/Animations/Nug_TurnLeft90_InMotion
   }
   n2 = Sequence Player 'Nug_TurnRight90_InMotion'() {
    bLoopAnimation=false
    groupName=Locomotion
    groupRole=TransitionFollower
    method=SyncGroup
    sequence=/Irwin/AI/Prey/Nug/Animations/Nug_TurnRight90_InMotion
   }
   Output Pose(Result=Blend Poses by bool(BlendPose_0=@n1, BlendPose_1=@n2, bActiveValue=Get bTurnAngleIsNegative() -> (bTurnAngleIsNegative:bool)) {blendType=Linear, childUpateMode=ResetChildOnActivate, transitionType=Inertialization}) {layerGroup=FaunaLocomotionOverride}
  ]}
]
variables [
 Delta Time X {type=number}
 FlipNeckToRight {type=boolean}
 IsPlayingAttack {type=boolean}
 NeckAlpha {type=number}
 NeckFlopRondom {type=integer}
 NeckLeft {type=boolean, default=true}
 TurnLeft {type=boolean}
 TurnOffNeckFlop {type=number}
 TurnRight {type=boolean}
]
