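# Sound cue for the NugZ creature's aggro pre-charge vocalization.
# A quality-level node sends the first two levels to a random pick of nine takes and the last two to a five-take subset, which is excluded from branch culling.
# It uses shared prey vocalization attenuation, a creature vocalization sound class, and two Nug voice concurrency sets for global and priority control.
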SoundCue /NugZ/Sounds/NugZ_Aggro_PreCharge_Cue
root=QualityLevel_0
nodes [
 {id=QualityLevel_0, children [Random_0,Random_0,Random_1,Random_1]}
 {
  id=Random_0
  children [WavePlayer_0,WavePlayer_1,WavePlayer_2,WavePlayer_3,WavePlayer_4,WavePlayer_5,WavePlayer_6,WavePlayer_7,WavePlayer_8]
  properties {weights [1,1,1,1,1,1,1,1,1]}
 }
 {
  id=Random_1
  children [WavePlayer_9, WavePlayer_10, WavePlayer_11, WavePlayer_13, WavePlayer_12]
  properties {bShouldExcludeFromBranchCulling=true, weights [1,1,1,1,1]}
 }
 {id=WavePlayer_0, properties {soundWaveAssetPtr=&/NugZ_Aggro_PreCharge_01}}
 {id=WavePlayer_1, properties {soundWaveAssetPtr=&/NugZ_Aggro_PreCharge_02}}
 {id=WavePlayer_10, properties {soundWaveAssetPtr=&NugZ_Aggro_PreCharge_02}}
 {id=WavePlayer_11, properties {soundWaveAssetPtr=&/NugZ_Aggro_PreCharge_05}}
 {id=WavePlayer_12, properties {soundWaveAssetPtr=&/NugZ_Aggro_PreCharge_09}}
 {id=WavePlayer_13, properties {soundWaveAssetPtr=&/NugZ_Aggro_PreCharge_07}}
 {id=WavePlayer_2, properties {soundWaveAssetPtr=&/NugZ_Aggro_PreCharge_03}}
 {id=WavePlayer_3, properties {soundWaveAssetPtr=&/NugZ_Aggro_PreCharge_04}}
 {id=WavePlayer_4, properties {soundWaveAssetPtr=&NugZ_Aggro_PreCharge_05}}
 {id=WavePlayer_5, properties {soundWaveAssetPtr=&/NugZ_Aggro_PreCharge_06}}
 {id=WavePlayer_6, properties {soundWaveAssetPtr=&NugZ_Aggro_PreCharge_07}}
 {id=WavePlayer_7, properties {soundWaveAssetPtr=&/NugZ_Aggro_PreCharge_08}}
 {id=WavePlayer_8, properties {soundWaveAssetPtr=&NugZ_Aggro_PreCharge_09}}
 {id=WavePlayer_9, properties {soundWaveAssetPtr=&NugZ_Aggro_PreCharge_01}}
]
properties {
 attenuationSettings=/Irwin/Sounds/Settings/Prey_Vocalization_Att
 maxDistance=6350
 priority=50
 soundClassObject=&Fort_Creatures_Vocalization_Tactical
 volumeMultiplier=1.6
 concurrencySet [/Irwin/Sounds/Settings/Conc/Nug_Global_VO_Conc, /Irwin/Sounds/Settings/Conc/Nug_SetPriority_VO_Conc]
}
