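# Sound cue for the NugZ creature's aggro-start vocalization.
# A Random node picks one of six equally weighted wave variations and feeds it through a Modulator at the root.
# It uses prey aggro attenuation and the tactical creature vocalization sound class.
# Its concurrency set combines global and idle Nug voice rules with a NugZ single prevent-new rule.
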SoundCue /NugZ/Sounds/NugZ_Aggro_Start_Cue
root=Modulator_0
nodes [
 {id=Modulator_0, children [Random_0]}
 {
  id=Random_0
  children [WavePlayer_0,WavePlayer_1,WavePlayer_2,WavePlayer_3,WavePlayer_4,WavePlayer_5]
  properties {bShouldExcludeFromBranchCulling=true, weights [1,1,1,1,1,1]}
 }
 {id=WavePlayer_0, properties {soundWaveAssetPtr=&/NugZ_Aggro_Start_06}}
 {id=WavePlayer_1, properties {soundWaveAssetPtr=&/NugZ_Aggro_Start_01}}
 {id=WavePlayer_2, properties {soundWaveAssetPtr=&/NugZ_Aggro_Start_02}}
 {id=WavePlayer_3, properties {soundWaveAssetPtr=&/NugZ_Aggro_Start_03}}
 {id=WavePlayer_4, properties {soundWaveAssetPtr=&/NugZ_Aggro_Start_04}}
 {id=WavePlayer_5, properties {soundWaveAssetPtr=&/NugZ_Aggro_Start_05}}
]
properties {
 attenuationSettings=/Irwin/Sounds/Settings/Prey_Aggro_Vocalization_Att
 maxDistance=7000
 priority=100
 soundClassObject=&Fort_Creatures_Vocalization_Tactical
 volumeMultiplier=1
 concurrencyOverrides {bLimitToOwner=true, maxCount=1, resolutionRule=PreventNew}
 concurrencySet [
  /Irwin/Sounds/Settings/Conc/Nug_Global_VO_Conc
  /Game/Sounds/Concurrency_Rules/NugZ_SinglePreventNew
  /Irwin/Sounds/Settings/Conc/Nug_Idle_VO_Conc
 ]
}
