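# Sound cue for the NugZ creature's attack vocalization, playing one randomly chosen, equally weighted attack clip.
# A default Modulator feeds a Quality Level switch: the first two levels draw from eleven clips, the last two from a smaller five-clip pool.
# The smaller pool is excluded from branch culling.
# Irwin prey vocalization attenuation, two Nug concurrency sets and a creature vocalization sound class govern playback.
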SoundCue /NugZ/Sounds/NugZ_Attack_Cue
root=Modulator_0
nodes [
 {id=Modulator_0, children [QualityLevel_0]}
 {id=QualityLevel_0, children [Random_1,Random_1,Random_2,Random_2]}
 {
  id=Random_1
  children [
   WavePlayer_15
   WavePlayer_16
   WavePlayer_17
   WavePlayer_18
   WavePlayer_19
   WavePlayer_20
   WavePlayer_21
   WavePlayer_22
   WavePlayer_23
   WavePlayer_24
   WavePlayer_25
  ]
  properties {weights [1,1,1,1,1,1,1,1,1,1,1]}
 }
 {
  id=Random_2
  children [WavePlayer_28, WavePlayer_27, WavePlayer_26, WavePlayer_29, WavePlayer_30]
  properties {bShouldExcludeFromBranchCulling=true, weights [1,1,1,1,1]}
 }
 {id=WavePlayer_15, properties {soundWaveAssetPtr=&/NugZ_Attack_01}}
 {id=WavePlayer_16, properties {soundWaveAssetPtr=&/NugZ_Attack_02}}
 {id=WavePlayer_17, properties {soundWaveAssetPtr=&/NugZ_Attack_03}}
 {id=WavePlayer_18, properties {soundWaveAssetPtr=&/NugZ_Attack_04}}
 {id=WavePlayer_19, properties {soundWaveAssetPtr=&/NugZ_Attack_05}}
 {id=WavePlayer_20, properties {soundWaveAssetPtr=&/NugZ_Attack_06}}
 {id=WavePlayer_21, properties {soundWaveAssetPtr=&/NugZ_Attack_07}}
 {id=WavePlayer_22, properties {soundWaveAssetPtr=&/NugZ_Attack_08}}
 {id=WavePlayer_23, properties {soundWaveAssetPtr=&/NugZ_Attack_09}}
 {id=WavePlayer_24, properties {soundWaveAssetPtr=&/NugZ_Attack_10}}
 {id=WavePlayer_25, properties {soundWaveAssetPtr=&/NugZ_Attack_11}}
 {id=WavePlayer_26, properties {soundWaveAssetPtr=&NugZ_Attack_02}}
 {id=WavePlayer_27, properties {soundWaveAssetPtr=&NugZ_Attack_04}}
 {id=WavePlayer_28, properties {soundWaveAssetPtr=&NugZ_Attack_05}}
 {id=WavePlayer_29, properties {soundWaveAssetPtr=&NugZ_Attack_07}}
 {id=WavePlayer_30, properties {soundWaveAssetPtr=&NugZ_Attack_09}}
]
properties {
 attenuationSettings=/Irwin/Sounds/Settings/Prey_Vocalization_Att
 maxDistance=6350
 soundClassObject=&Fort_Creatures_Vocalization_Tactical
 volumeMultiplier=1
 concurrencySet [/Irwin/Sounds/Settings/Conc/Nug_Global_VO_Conc, /Irwin/Sounds/Settings/Conc/Nug_SetPriority_VO_Conc]
}
