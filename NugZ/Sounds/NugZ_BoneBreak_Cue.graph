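# Sound Cue that plays a bone-break sound for NugZ, routed through the creature foley sound class.
# A Random node at the root picks one of five equally weighted wave variations and is excluded from branch culling.
# It reuses the prey footstep attenuation from the Irwin content and plays at slightly reduced volume.
# Overridden concurrency caps simultaneous plays per owner, blocking new ones once the limit is reached.
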SoundCue /NugZ/Sounds/NugZ_BoneBreak_Cue
root=Random_0
nodes [
 {
  id=Random_0
  children [WavePlayer_0,WavePlayer_1,WavePlayer_2,WavePlayer_3,WavePlayer_4]
  properties {bShouldExcludeFromBranchCulling=true, weights [1,1,1,1,1]}
 }
 {id=WavePlayer_0, properties {soundWaveAssetPtr=&/NugZ_BoneBreak_01}}
 {id=WavePlayer_1, properties {soundWaveAssetPtr=&/NugZ_BoneBreak_02}}
 {id=WavePlayer_2, properties {soundWaveAssetPtr=&/NugZ_BoneBreak_03}}
 {id=WavePlayer_3, properties {soundWaveAssetPtr=&/NugZ_BoneBreak_04}}
 {id=WavePlayer_4, properties {soundWaveAssetPtr=&/NugZ_BoneBreak_05}}
]
properties {
 attenuationSettings=/Irwin/Sounds/Settings/Prey_Step_Att
 bOverrideConcurrency=true
 maxDistance=3200
 soundClassObject=/Game/Sounds/FortSoundClassesAndModes/Fort_Creatures_Foley
 volumeMultiplier=0.8
 concurrencyOverrides {bLimitToOwner=true, maxCount=3, resolutionRule=PreventNew}
}
