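# NugZ_Eat_Cue is a Sound Cue that plays an eating vocalization for the NugZ creature.
# A Random node picks one of four equally weighted eat waves, excluded from branch culling, and feeds a Modulator at the root.
# It uses shared Irwin prey vocalization attenuation, a tactical creature vocalization sound class, and two Nug voice concurrency sets for global and priority-based limiting.
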SoundCue /NugZ/Sounds/NugZ_Eat_Cue
root=Modulator_0
nodes [
 {id=Modulator_0, children [Random_0]}
 {
  id=Random_0
  children [WavePlayer_0,WavePlayer_1,WavePlayer_2,WavePlayer_3]
  properties {bShouldExcludeFromBranchCulling=true, weights [1,1,1,1]}
 }
 {id=WavePlayer_0, properties {soundWaveAssetPtr=&/NugZ_Eat_01}}
 {id=WavePlayer_1, properties {soundWaveAssetPtr=&/NugZ_Eat_02}}
 {id=WavePlayer_2, properties {soundWaveAssetPtr=&/NugZ_Eat_03}}
 {id=WavePlayer_3, properties {soundWaveAssetPtr=&/NugZ_Eat_04}}
]
properties {
 attenuationSettings=/Irwin/Sounds/Settings/Prey_Vocalization_Att
 maxDistance=6350
 priority=4
 soundClassObject=&Fort_Creatures_Vocalization_Tactical
 volumeMultiplier=1
 concurrencySet [/Irwin/Sounds/Settings/Conc/Nug_Global_VO_Conc, /Irwin/Sounds/Settings/Conc/Nug_SetPriority_VO_Conc]
}
