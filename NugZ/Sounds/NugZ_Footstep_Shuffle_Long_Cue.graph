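# Sound cue for NugZ long shuffle footstep sounds.
# A random node picks one of six shuffle wave variants with equal weight.
# It plays at a reduced volume through the Fortnite creature steps sound class, using the Irwin prey step attenuation for distance falloff.
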SoundCue /NugZ/Sounds/NugZ_Footstep_Shuffle_Long_Cue
root=Random_0
nodes [
 {id=Random_0, children [WavePlayer_0,WavePlayer_1,WavePlayer_2,WavePlayer_3,WavePlayer_4,WavePlayer_5], properties {weights [1,1,1,1,1,1]}}
 {id=WavePlayer_0, properties {soundWaveAssetPtr=&/NugZ_Footstep_Shuffle_Long_01}}
 {id=WavePlayer_1, properties {soundWaveAssetPtr=&/NugZ_Footstep_Shuffle_Long_02}}
 {id=WavePlayer_2, properties {soundWaveAssetPtr=&/NugZ_Footstep_Shuffle_Long_03}}
 {id=WavePlayer_3, properties {soundWaveAssetPtr=&/NugZ_Footstep_Shuffle_Long_04}}
 {id=WavePlayer_4, properties {soundWaveAssetPtr=&/NugZ_Footstep_Shuffle_Long_05}}
 {id=WavePlayer_5, properties {soundWaveAssetPtr=&/NugZ_Footstep_Shuffle_Long_06}}
]
properties {
 attenuationSettings=/Irwin/Sounds/Settings/Prey_Step_Att
 maxDistance=3200
 soundClassObject=/Game/Sounds/FortSoundClassesAndModes/Fort_Creatures_Steps_Normal
 volumeMultiplier=0.2
}
