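# Sound cue for NugZ idle clucks, a quiet, low-priority creature vocalization.
# A quality switch sends the first two levels to a random pick of nine clucks plus four empty slots for occasional silence.
# The last two levels use a five-cluck subset excluded from branch culling, with no empty slots.
# Shared prey attenuation and global plus idle Nug concurrency sets control falloff and voice count.
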SoundCue /NugZ/Sounds/NugZ_Idle_Cluck_Cue
root=QualityLevel_1
nodes [
 {id=QualityLevel_1, children [Random_0,Random_0,Random_1,Random_1]}
 {
  id=Random_0
  children [WavePlayer_1,WavePlayer_2,WavePlayer_3,WavePlayer_4,WavePlayer_5,WavePlayer_6,WavePlayer_7,WavePlayer_8,WavePlayer_9,null,null,null,null]
  properties {weights [1,1,1,1,1,1,1,1,1,1,1,1,1]}
 }
 {
  id=Random_1
  children [WavePlayer_19, WavePlayer_17, WavePlayer_16, WavePlayer_15, WavePlayer_14]
  properties {bShouldExcludeFromBranchCulling=true, weights [1,1,1,1,1]}
 }
 {id=WavePlayer_1, properties {soundWaveAssetPtr=&/NugZ_Idle_Cluck_01}}
 {id=WavePlayer_14, properties {soundWaveAssetPtr=&/NugZ_Idle_Cluck_06}}
 {id=WavePlayer_15, properties {soundWaveAssetPtr=&/NugZ_Idle_Cluck_05}}
 {id=WavePlayer_16, properties {soundWaveAssetPtr=&/NugZ_Idle_Cluck_04}}
 {id=WavePlayer_17, properties {soundWaveAssetPtr=&/NugZ_Idle_Cluck_02}}
 {id=WavePlayer_19, properties {soundWaveAssetPtr=&/NugZ_Idle_Cluck_09}}
 {id=WavePlayer_2, properties {soundWaveAssetPtr=&NugZ_Idle_Cluck_02}}
 {id=WavePlayer_3, properties {soundWaveAssetPtr=&/NugZ_Idle_Cluck_03}}
 {id=WavePlayer_4, properties {soundWaveAssetPtr=&NugZ_Idle_Cluck_04}}
 {id=WavePlayer_5, properties {soundWaveAssetPtr=&NugZ_Idle_Cluck_05}}
 {id=WavePlayer_6, properties {soundWaveAssetPtr=&NugZ_Idle_Cluck_06}}
 {id=WavePlayer_7, properties {soundWaveAssetPtr=&/NugZ_Idle_Cluck_07}}
 {id=WavePlayer_8, properties {soundWaveAssetPtr=&/NugZ_Idle_Cluck_08}}
 {id=WavePlayer_9, properties {soundWaveAssetPtr=&NugZ_Idle_Cluck_09}}
]
properties {
 attenuationSettings=/Irwin/Sounds/Settings/Prey_Vocalization_Att
 maxDistance=6350
 priority=0.2
 soundClassObject=/Game/Sounds/FortSoundClassesAndModes/Fort_Creatures_Vocalization_Normal
 volumeMultiplier=0.5
 concurrencySet [/Irwin/Sounds/Settings/Conc/Nug_Global_VO_Conc, /Irwin/Sounds/Settings/Conc/Nug_Idle_VO_Conc]
}
