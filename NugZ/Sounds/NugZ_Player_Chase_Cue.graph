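# Sound cue for the NugZ creature's player-chase vocalizations, routed to the tactical creature vocalization sound class.
# A Modulator feeds a quality-level switch: the first two levels pick randomly from twelve equally weighted waves, the last two from a five-wave subset.
# The smaller random node is excluded from branch culling.
# Shared prey vocalization attenuation and three concurrency rules (global, idle, single-prevent-new) control falloff and limit overlapping voices.
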SoundCue /NugZ/Sounds/NugZ_Player_Chase_Cue
root=Modulator_0
nodes [
 {id=Modulator_0, children [QualityLevel_0]}
 {id=QualityLevel_0, children [Random_0,Random_0,Random_1,Random_1]}
 {
  id=Random_0
  children [
   WavePlayer_0
   WavePlayer_1
   WavePlayer_2
   WavePlayer_3
   WavePlayer_4
   WavePlayer_5
   WavePlayer_6
   WavePlayer_7
   WavePlayer_8
   WavePlayer_9
   WavePlayer_10
   WavePlayer_11
  ]
  properties {weights [1,1,1,1,1,1,1,1,1,1,1,1]}
 }
 {
  id=Random_1
  children [WavePlayer_12, WavePlayer_13, WavePlayer_14, WavePlayer_15, WavePlayer_16]
  properties {bShouldExcludeFromBranchCulling=true, weights [1,1,1,1,1]}
 }
 {id=WavePlayer_0, properties {soundWaveAssetPtr=&/NugZ_Player_Chase_01}}
 {id=WavePlayer_1, properties {soundWaveAssetPtr=&/NugZ_Player_Chase_02}}
 {id=WavePlayer_10, properties {soundWaveAssetPtr=&/NugZ_Player_Chase_11}}
 {id=WavePlayer_11, properties {soundWaveAssetPtr=&/NugZ_Player_Chase_12}}
 {id=WavePlayer_12, properties {soundWaveAssetPtr=&NugZ_Player_Chase_01}}
 {id=WavePlayer_13, properties {soundWaveAssetPtr=&NugZ_Player_Chase_02}}
 {id=WavePlayer_14, properties {soundWaveAssetPtr=&/NugZ_Player_Chase_06}}
 {id=WavePlayer_15, properties {soundWaveAssetPtr=&/NugZ_Player_Chase_08}}
 {id=WavePlayer_16, properties {soundWaveAssetPtr=&NugZ_Player_Chase_11}}
 {id=WavePlayer_2, properties {soundWaveAssetPtr=&/NugZ_Player_Chase_03}}
 {id=WavePlayer_3, properties {soundWaveAssetPtr=&/NugZ_Player_Chase_04}}
 {id=WavePlayer_4, properties {soundWaveAssetPtr=&/NugZ_Player_Chase_05}}
 {id=WavePlayer_5, properties {soundWaveAssetPtr=&NugZ_Player_Chase_06}}
 {id=WavePlayer_6, properties {soundWaveAssetPtr=&/NugZ_Player_Chase_07}}
 {id=WavePlayer_7, properties {soundWaveAssetPtr=&NugZ_Player_Chase_08}}
 {id=WavePlayer_8, properties {soundWaveAssetPtr=&/NugZ_Player_Chase_09}}
 {id=WavePlayer_9, properties {soundWaveAssetPtr=&/NugZ_Player_Chase_10}}
]
properties {
 attenuationSettings=/Irwin/Sounds/Settings/Prey_Vocalization_Att
 maxDistance=6350
 priority=2
 soundClassObject=&Fort_Creatures_Vocalization_Tactical
 volumeMultiplier=1
 concurrencySet [
  /Irwin/Sounds/Settings/Conc/Nug_Global_VO_Conc
  /Game/Sounds/Concurrency_Rules/NugZ_SinglePreventNew
  /Irwin/Sounds/Settings/Conc/Nug_Idle_VO_Conc
 ]
}
