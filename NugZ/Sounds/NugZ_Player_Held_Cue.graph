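# Sound cue for NugZ creature vocalizations, named for when a player holds one, using prey vocalization attenuation and a creature vocalization sound class.
# A quality-level node routes two levels to an evenly weighted random pick of fifteen held waves.
# The other two levels use a six-wave subset of the same recordings, whose random node is excluded from branch culling.
# Three Nug concurrency sets govern priority and overlap with held and held-attack voices.
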SoundCue /NugZ/Sounds/NugZ_Player_Held_Cue
root=QualityLevel_0
nodes [
 {id=QualityLevel_0, children [Random_0,Random_0,Random_1,Random_1]}
 {
  id=Random_0
  children [
   WavePlayer_0
   WavePlayer_1
   WavePlayer_2
   WavePlayer_3
   WavePlayer_4
   WavePlayer_5
   WavePlayer_6
   WavePlayer_7
   WavePlayer_8
   WavePlayer_9
   WavePlayer_10
   WavePlayer_11
   WavePlayer_12
   WavePlayer_13
   WavePlayer_14
  ]
  properties {weights [1,1,1,1,1,1,1,1,1,1,1,1,1,1,1]}
 }
 {
  id=Random_1
  children [WavePlayer_15, WavePlayer_17, WavePlayer_16, WavePlayer_18, WavePlayer_19, WavePlayer_20]
  properties {bShouldExcludeFromBranchCulling=true, weights [1,1,1,1,1,1]}
 }
 {id=WavePlayer_0, properties {soundWaveAssetPtr=&/NugZ_Player_Held_01}}
 {id=WavePlayer_1, properties {soundWaveAssetPtr=&/NugZ_Player_Held_02}}
 {id=WavePlayer_10, properties {soundWaveAssetPtr=&/NugZ_Player_Held_11}}
 {id=WavePlayer_11, properties {soundWaveAssetPtr=&/NugZ_Player_Held_12}}
 {id=WavePlayer_12, properties {soundWaveAssetPtr=&/NugZ_Player_Held_13}}
 {id=WavePlayer_13, properties {soundWaveAssetPtr=&/NugZ_Player_Held_14}}
 {id=WavePlayer_14, properties {soundWaveAssetPtr=&/NugZ_Player_Held_15}}
 {id=WavePlayer_15, properties {soundWaveAssetPtr=&NugZ_Player_Held_01}}
 {id=WavePlayer_16, properties {soundWaveAssetPtr=&/NugZ_Player_Held_03}}
 {id=WavePlayer_17, properties {soundWaveAssetPtr=&/NugZ_Player_Held_06}}
 {id=WavePlayer_18, properties {soundWaveAssetPtr=&/NugZ_Player_Held_09}}
 {id=WavePlayer_19, properties {soundWaveAssetPtr=&NugZ_Player_Held_12}}
 {id=WavePlayer_2, properties {soundWaveAssetPtr=&NugZ_Player_Held_03}}
 {id=WavePlayer_20, properties {soundWaveAssetPtr=&NugZ_Player_Held_14}}
 {id=WavePlayer_3, properties {soundWaveAssetPtr=&/NugZ_Player_Held_04}}
 {id=WavePlayer_4, properties {soundWaveAssetPtr=&/NugZ_Player_Held_05}}
 {id=WavePlayer_5, properties {soundWaveAssetPtr=&NugZ_Player_Held_06}}
 {id=WavePlayer_6, properties {soundWaveAssetPtr=&/NugZ_Player_Held_07}}
 {id=WavePlayer_7, properties {soundWaveAssetPtr=&/NugZ_Player_Held_08}}
 {id=WavePlayer_8, properties {soundWaveAssetPtr=&NugZ_Player_Held_09}}
 {id=WavePlayer_9, properties {soundWaveAssetPtr=&/NugZ_Player_Held_10}}
]
properties {
 attenuationSettings=/Irwin/Sounds/Settings/Prey_Vocalization_Att
 maxDistance=6350
 soundClassObject=/Game/Sounds/FortSoundClassesAndModes/Fort_Creatures_Vocalization_Normal
 volumeMultiplier=0.7
 concurrencySet [
  /Irwin/Sounds/Settings/Conc/Nug_SetPriority_VO_Conc
  /Irwin/Sounds/Settings/Conc/Nug_Held_VO_Conc
  /Irwin/Sounds/Settings/Conc/Nug_HeldAttack_VO_Conc
 ]
}
