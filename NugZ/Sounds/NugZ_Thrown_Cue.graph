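# Sound cue for a NugZ being thrown, layering a throw foley with a delayed thrown sound under creature vocalization class and prey attenuation.
# Each mixer plays one of three foley waves at double input volume, plus a randomly chosen thrown wave after a short random delay.
# A quality switch gives the first two levels eight thrown variants and the last two a four-variant subset, whose random node is excluded from branch culling.
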SoundCue /NugZ/Sounds/NugZ_Thrown_Cue
root=QualityLevel_0
nodes [
 {id=Delay_2, children [Random_0], properties {delayMax=0.2, delayMin=0.15}}
 {id=Delay_3, children [Random_1], properties {delayMax=0.2, delayMin=0.15}}
 {id=Mixer_0, children [Delay_2,Random_2], properties {inputVolume [1,2]}}
 {id=Mixer_1, children [Random_2,Delay_3], properties {inputVolume [2,1]}}
 {id=QualityLevel_0, children [Mixer_0,Mixer_0,Mixer_1,Mixer_1]}
 {
  id=Random_0
  children [WavePlayer_0, WavePlayer_1, WavePlayer_2, WavePlayer_3, WavePlayer_4, WavePlayer_5, WavePlayer_6, WavePlayer_11]
  properties {weights [1,1,1,1,1,1,1,1]}
 }
 {
  id=Random_1
  children [WavePlayer_9, WavePlayer_8, WavePlayer_7, WavePlayer_10]
  properties {bShouldExcludeFromBranchCulling=true, weights [1,1,1,1]}
 }
 {id=Random_2, children [WavePlayer_12, WavePlayer_13, WavePlayer_14], properties {weights [1,1,1]}}
 {id=WavePlayer_0, properties {soundWaveAssetPtr=&/NugZ_Thrown_01}}
 {id=WavePlayer_1, properties {soundWaveAssetPtr=&/NugZ_Thrown_02}}
 {id=WavePlayer_10, properties {soundWaveAssetPtr=&/NugZ_Thrown_07}}
 {id=WavePlayer_11, properties {soundWaveAssetPtr=&/NugZ_Thrown_08}}
 {id=WavePlayer_12, properties {soundWaveAssetPtr=&/NugZ_Throw_Foley_03}}
 {id=WavePlayer_13, properties {soundWaveAssetPtr=&/NugZ_Throw_Foley_01}}
 {id=WavePlayer_14, properties {soundWaveAssetPtr=&/NugZ_Throw_Foley_02}}
 {id=WavePlayer_2, properties {soundWaveAssetPtr=&/NugZ_Thrown_03}}
 {id=WavePlayer_3, properties {soundWaveAssetPtr=&/NugZ_Thrown_04}}
 {id=WavePlayer_4, properties {soundWaveAssetPtr=&/NugZ_Thrown_05}}
 {id=WavePlayer_5, properties {soundWaveAssetPtr=&/NugZ_Thrown_06}}
 {id=WavePlayer_6, properties {soundWaveAssetPtr=&NugZ_Thrown_07}}
 {id=WavePlayer_7, properties {soundWaveAssetPtr=&NugZ_Thrown_04}}
 {id=WavePlayer_8, properties {soundWaveAssetPtr=&NugZ_Thrown_03}}
 {id=WavePlayer_9, properties {soundWaveAssetPtr=&NugZ_Thrown_02}}
]
properties {
 attenuationSettings=/Irwin/Sounds/Settings/Prey_Vocalization_Att
 maxDistance=6350
 soundClassObject=/Game/Sounds/FortSoundClassesAndModes/Fort_Creatures_Vocalization_Normal
 volumeMultiplier=1.3
}
